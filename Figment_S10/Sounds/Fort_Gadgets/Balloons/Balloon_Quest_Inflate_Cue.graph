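# Sound cue for a balloon inflation sound, named for a quest use among the balloon gadget sounds.
# A single long inflate wave plays through a modulator that varies pitch on each play, at reduced cue volume.
# Gadget-activate attenuation makes it spatial, and it is routed to a weapon projectile sound class.
# A concurrency rule allows one instance per owner, stopping the oldest.
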SoundCue /Figment_S10/Sounds/Fort_Gadgets/Balloons/Balloon_Quest_Inflate_Cue
root=Modulator_0
nodes [
 {id=Modulator_0, children [WavePlayer_0], properties {pitchMax=1.2}}
 {id=WavePlayer_0, properties {soundWaveAssetPtr=&/Balloon_Inflate_Long_01}}
]
properties {
 attenuationSettings=&Fort_GadgetActivate_Att
 maxDistance=4000
 soundClassObject=/Figment_S10/Sounds/FortSoundClassesAndModes/Fort_Weapon_Projectile_PSM
 volumeMultiplier=0.5
 concurrencySet [/Game/Sounds/Concurrency_Rules/OneMaximum_Owner_StopOldest]
}
